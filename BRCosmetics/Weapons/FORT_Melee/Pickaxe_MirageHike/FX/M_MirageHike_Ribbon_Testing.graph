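# M_MirageHike_Ribbon_Testing is a translucent, unlit test material for Niagara ribbons in the MirageHike pickaxe effects folder.
# Its second UV channel (index 1) drives a shared UV checker texture sample and a float2 component breakout.
# No connections from these nodes to material outputs are shown, so it appears to be a scratch graph for inspecting ribbon UVs.
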Material /BRCosmetics/Weapons/FORT_Melee/Pickaxe_MirageHike/FX/M_MirageHike_Ribbon_Testing
graph {
 properties {bUsedWithNiagaraRibbons=true, blendMode=BLEND_Translucent, pixelDepthOffsetMode=PDOM_Legacy, shadingModel=MSM_Unlit, usedShadingModels=Unlit}
 nodes [
  n1 = TextureCoordinate() {coordinateIndex=1}
  MaterialFunctionCall(Float2=@n1) {materialFunction=&BreakOutFloat2Components}
  TextureSample(UVs=@n1) {texture=/Game/Packages/Fortress_SharedMaterials/Textures/T_UV_Checker_D}
 ]
}
