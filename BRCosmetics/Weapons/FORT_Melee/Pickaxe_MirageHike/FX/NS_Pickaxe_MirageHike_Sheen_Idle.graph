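# Niagara system for the idle sheen effect on the MirageHike pickaxe, pairing a mesh sheen sweep with a sparkle.
# Sheen_Mesh bursts a mesh living for the system-level SheenLifetime, animating a dynamic material parameter from 0 to 1 via curve.
# Sparkle bursts shortly before the sheen ends, timed from that same variable.
# It is a short-lived, camera-offset, hue-shifting sprite at a fixed position, spinning and scaling in and out.
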NiagaraSystem /BRCosmetics/Weapons/FORT_Melee/Pickaxe_MirageHike/FX/NS_Pickaxe_MirageHike_Sheen_Idle
dependencies {
 usedDataInterfaces [&NiagaraDataInterfaceCurve]
 usedDynamicInputs []
 usedModules [
  &SystemState
  &self:SystemSpawnScript.SystemScriptSource.SystemScriptGraph.NiagaraNodeAssignment_0.SetVariables_#1
 ]
}
emitters {
 Sheen_Mesh {
  emitterUpdateScript [
   &EmitterState()
   &SpawnBurst_Instantaneous(Spawn Time=0.02, Use Spawn Probability=false, UseLoopCountLimit=false)
  ]
  particleSpawnScript [
   &InitializeParticle(
    Color=<rgba 1.6458 1.8399 2 0.1>
    Lifetime=$System.SheenLifetime
    Mesh Scale Mode=::NewEnumerator3
    UsePositionOffset=false
    Write Material Random=true
   )
   &InitialMeshOrientation(Mesh Axis To Orient=<0 1 0>, Mesh Orientation Mode=::NewEnumerator2, Use Rotation Vector=false)
  ]
  particleUpdateScript [
   &ParticleState()
   &DynamicMaterialParameters(
    Index 0 Param 1=&FloatFromCurve(FloatCurve=DataInterface() {propertyValues {Curve {keys [{interpMode=RCIM_Cubic, tangentMode=RCTM_User, time=0, value=0}, {interpMode=RCIM_Cubic, tangentMode=RCTM_User, time=1, value=1}]}}})
    Index 0 Param 2=&RandomRangeFloat(Minimum=0.3, Maximum=0.4)
    Index 0 Param 3=0.025
    Module.Param0WriteEnabled=true
    Module.Param1WriteEnabled=true
    Module.Param2WriteEnabled=true
    Module.Param3WriteEnabled=false
   )
  ]
  rendererClasses [&NiagaraMeshRendererProperties]
  renderers [{rendererClass=&NiagaraMeshRendererProperties, rendererIndex=0}]
 }
 Sparkle {
  emitterUpdateScript [
   &EmitterState()
   &SpawnBurst_Instantaneous(Spawn Time=&Add_Float(A=$System.SheenLifetime, B=-0.16), Use Spawn Probability=false, UseLoopCountLimit=false)
  ]
  particleSpawnScript [
   &InitializeParticle(
    Lifetime=0.25
    Sprite Rotation Mode=::NewEnumerator2
    Sprite Size Mode=::NewEnumerator1
    Uniform Sprite Size Max=47.5
    Uniform Sprite Size Min=40
    UsePositionOffset=false
    Write Material Random=true
   )
   &CameraOffset(Camera Offset Amount=20)
  ]
  particleUpdateScript [
   &ParticleState()
   &ScaleSpriteSize(Uniform Curve Sprite Scale=DataInterface() {propertyValues {Curve {keys [
        <at 0 0 RCIM_Cubic>
        {arriveTangent=0.547, interpMode=RCIM_Cubic, leaveTangent=0.547, tangentMode=RCTM_User, time=0.1, value=0.75}
        {arriveTangent=-0.8333, interpMode=RCIM_Cubic, leaveTangent=-0.8333, time=0.8, value=0.9}
        {arriveTangent=-13.8944, interpMode=RCIM_Cubic, leaveTangent=-13.8944, tangentMode=RCTM_User, time=1, value=0}
       ]}}})
   &SpriteRotationRate(Rotation Rate=&RandomRangeFloat(Minimum=150, Maximum=200))
   &Color(Color=&HueShiftLinearColor(
     Hue=&FloatFromCurve(FloatCurve=DataInterface() {propertyValues {Curve {keys [{tangentMode=RCTM_User, time=0, value=0.4}, {tangentMode=RCTM_User, time=1, value=4}]}}})
     Saturation=1
     Value=&RandomRangeFloat(Minimum=0.2, Maximum=0.5)
     Linear Color=<rgba 1.569 1.5208 2 1>
    ))
   /SetVariables(Particles.Position=&MakeVector(X=-0.25, Y=0, Z=97)) {isSetParameters=true}
  ]
  rendererClasses [&NiagaraSpriteRendererProperties]
  renderers [{rendererClass=&NiagaraSpriteRendererProperties}]
 }
}
summary {
 systemName=NS_Pickaxe_MirageHike_Sheen_Idle
 emitters [
  {emitterName=Sheen_Mesh, rendererClasses [&NiagaraMeshRendererProperties]}
  {emitterName=Sparkle, rendererClasses [&NiagaraSpriteRendererProperties]}
 ]
 userVariables []
}
